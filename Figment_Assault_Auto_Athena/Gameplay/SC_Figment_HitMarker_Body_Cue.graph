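# Sound cue for the body-hit marker sound of the Figment automatic assault rifle.
# A Random node picks one of three equally weighted shared HitMarker_Body waves and feeds it through a Modulator left at its default settings.
# Playback is boosted in volume, uses bullet-impact attenuation and the Impacts sound class, and joins a hit-notification concurrency set named for a two-voice, stop-oldest limit.
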SoundCue /Figment_Assault_Auto_Athena/Gameplay/SC_Figment_HitMarker_Body_Cue
root=Modulator_0
nodes [
 {id=Modulator_0, children [Random_0]}
 {id=Random_0, children [WavePlayer_0,WavePlayer_1,WavePlayer_2], properties {bShouldExcludeFromBranchCulling=true, weights [1,1,1]}}
 {id=WavePlayer_0, properties {soundWaveAssetPtr=/Game/Athena/Sounds/HitNotifications/HitMarker_Body_01}}
 {id=WavePlayer_1, properties {soundWaveAssetPtr=/Game/Athena/Sounds/HitNotifications/HitMarker_Body_02}}
 {id=WavePlayer_2, properties {soundWaveAssetPtr=/Game/Athena/Sounds/HitNotifications/HitMarker_Body_03}}
]
properties {
 attenuationSettings=/Game/Sounds/Attenuation/Fort_Bullet_Impact_Att
 maxDistance=5300
 soundClassObject=&Fort_Impacts
 volumeMultiplier=1.5
 concurrencyOverrides {maxCount=2, resolutionRule=StopOldest}
 concurrencySet [&HitNotifications_TwoMaximum_StopOldest_VolumeScale]
}
